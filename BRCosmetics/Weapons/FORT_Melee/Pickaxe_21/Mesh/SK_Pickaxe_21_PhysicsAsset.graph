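# Physics asset for the Pickaxe_21 melee weapon skeletal mesh, a cosmetic pickaxe under BRCosmetics.
# It defines a single body on the b_AR_Root bone, made of one large capsule, rotated and slightly offset from the bone origin.
# It has no constraints, so the pickaxe collides and simulates as one rigid shape with no articulated parts.
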PhysicsAsset /BRCosmetics/Weapons/FORT_Melee/Pickaxe_21/Mesh/SK_Pickaxe_21_PhysicsAsset
bodies {b_AR_Root [{shapeType=Capsule, length=69.9224, radius=36.8869, center=<11.431 -0.7935 -2.8943>, rotation=<rot -0.0045 -90 -2.1673>}]}
constraints []
